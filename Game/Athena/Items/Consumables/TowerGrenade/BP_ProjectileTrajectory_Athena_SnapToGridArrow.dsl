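; Trajectory arrow for the Athena TowerGrenade consumable that previews the building snapped to the building grid at the aimed landing point.
; Each spline update uses the cached hit result and last point to position the preview, facing the instigator's cardinal direction.
; A door-area box trace selects the normal or underground mesh.
; Bounce-object hits invalidate the trajectory; visibility follows validity, plus windup state on touch.
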(event EventBeginPlay
  (|CallParentFunction))

(event EventPreSetTrajectorySpline
  (params (SplinePoints :type Array%20of%20Vectors) (SplineTangents :type Array%20of%20Vectors) (TrajectoryHitResult :type Hit%20Result%20Structure))
  ; Only valid in functions PreSetTrajectorySpline and SetTrajectorySpline!
  (Variables|Trajectory|SetCache%20-%20Trajectory%20Hit%20Result%20 TrajectoryHitResult))

(event Trajectory|EventSetTrajectorySpline
  (params (SplinePoints :type Array%20of%20Vectors) (SplineTangents :type Array%20of%20Vectors))
  (bind _cache___trajectory_hit_result (Variables|Trajectory|GetCache%20-%20Trajectory%20Hit%20Result%20))
  (bind (_bblockinghit _binitialoverlap _time _distance _location _impactpoint _normal _impactnormal _physmat _hitactor _hitcomponent) (Collision|BreakHitResult _cache___trajectory_hit_result))
  (Utilities|FlowControl|Sequence
    (:then_0
      ;
      (|CallParentFunction SplinePoints SplineTangents)
      (if (> (Utilities|Array|Length SplinePoints) 0)
        (CallFunction|HandleGridSnapUpdates (Utilities|Array|Get(aref) SplinePoints (Utilities|Array|LastIndex SplinePoints)) _cache___trajectory_hit_result)
        (bind _returnvalue (Default|ShouldBounceOffObject _hitactor _hitcomponent _normal))
        (|CallParentFunction (not _returnvalue))
        (else
          (Fortnite|FortShippingLog self "BP_ProjectileTrajectory_Athena_SnapToGridArrow: Tried to call SetTrajectorySpline with a spline points array with zero items. Unable handle snap to grid updates as the impact location of the trajectory is unknown." true))))
    (:then_1
      ; Reset after all logic completes.
      (Variables|Trajectory|SetCache%20-%20Trajectory%20Hit%20Result%20))))

(event EventSetTrajectoryValid
  (params (bIsTrajectoryValid :type Boolean)))

(fn CalculateDoorTraceLocation
  (params (Base%20Location :type Vector) (Cardinal%20Direction :type ECardinalDirection%20Enum%20%28by%20ref%29 :default "NewEnumerator0") (Cardinal%20Direction%20Yaw :type Float%20%28single-precision%29) (Rotatable%20Location%20Offset :type Vector))
  (return (Math|Vector|vector+vector Base%20Location (Math|Vector|RotateVectorAroundAxis Rotatable%20Location%20Offset Cardinal%20Direction%20Yaw "0, 0,1.000000") (Utilities|Map|Find (Variables|Default|GetOrientationDoortraceOffsets) Cardinal%20Direction))))

(fn HandleGridSnapUpdates
  (params (NewLocation :type Vector) (TrajectoryHitResult :type Hit%20Result%20Structure))
  (bind _sm_buildingpreview (Variables|Default|GetSM_BuildingPreview))
  (Networking|CanExecuteCosmeticEvents
    (:True
      ;[0.134 1 0.02478] Calculate Offsets
      (bind _output_get (LocalVariables|SetGridSnapLocation (Fortnite|SnaptoFortBuildingGrid NewLocation (Math|Vector|MakeVector 0.0 0.0 (Variables|SnaptoGrid|GetGridZOffset)) (Math|Vector|MakeVector 0.0 (Variables|SnaptoGrid|GetGridYOffset) (* (Variables|SnaptoGrid|GetGridZOffset) -1.0)))))
      (bind (_outcardinaldirection _outyaw) (Class|AthenaConsumablesSharedFunctionLibrary|GetActorForwardCardinalDirection (Game|GetInstigator)))
      (LocalVariables|SetCardinalDirection _outcardinaldirection)
      (LocalVariables|SetYaw%20Offset _outyaw)
      ;[0.9814 1 0] ONLY continue if a change occurred.
      (if (or (Math|Vector|NotEqual(Vector) _output_get (Variables|Default|GetLastGridSnapPosition) 0.0001) (Math|Float|NearlyEqual(Float) _outyaw (Variables|Default|GetLastYawOffset) 0.0001))
        (Utilities|FlowControl|Sequence
          (:then_0
            ;[0.1754 0.9502 1] Save off offsets so we can detect future changes.
            (Variables|Default|SetLastGridSnapPosition (LocalVariables|GetGridSnapLocation))
            (Variables|Default|SetLastYawOffset (LocalVariables|GetYaw%20Offset)))
          (:then_1
            ;[1 0.7127 0] Figure out if we hit the floor i.e. we need to use a different mesh to avoid essential parts (like doors) intersecting the ground.
            (Variables|Default|SetHitTerrain false)
            (bind _outhits (CallFunction|PerformDoorTrace (+ (Variables|Default|GetDoorTraceOffsetHigh) (CallFunction|CalculateDoorTraceLocation (LocalVariables|GetGridSnapLocation) (LocalVariables|GetCardinalDirection) (LocalVariables|GetYaw%20Offset) (Variables|Default|GetDoor%20Trace%20Rotatable%20Location%20Offset%20Const))) (+ (CallFunction|CalculateDoorTraceLocation (LocalVariables|GetGridSnapLocation) (LocalVariables|GetCardinalDirection) (LocalVariables|GetYaw%20Offset) (Variables|Default|GetDoor%20Trace%20Rotatable%20Location%20Offset%20Const)) (Variables|Default|GetDoorTraceOffsetLow)) (Variables|Default|GetDoorTraceHalfSize%20Const) (Variables|Default|GetDoorTraceDebugDrawTime) (select (<= (Variables|Default|GetDoorTraceDebugDrawTime) 0) (select (== (Variables|Default|GetDoorTraceDebugDrawTime) 0) "None" "Persistent") "ForDuration")))
            (bind _array_element (Utilities|Array|ForEachLoopwithBreak _outhits)
              (:LoopBody
                (if (Math|Boolean|ORBoolean (Utilities|IsA (.HitActor (Collision|BreakHitResult _array_element)) "/Script/Landscape.Landscape") (Utilities|IsA (.HitActor (Collision|BreakHitResult _array_element)) "/Script/Landscape.LandscapeStreamingProxy") (Utilities|IsA (.HitActor (Collision|BreakHitResult _array_element)) "/Script/FortniteGame.BuildingStaticGeo") (Utilities|IsA (.HitActor (Collision|BreakHitResult _array_element)) "/Script/FortniteGame.FortStaticMeshActor"))
                  (Variables|Default|SetHitTerrain true))
                (bind _asbuilding_actor (Utilities|Casting|CastToBuildingActor (.HitActor (Collision|BreakHitResult _array_element))))
                (Utilities|IsNotAVolume _asbuilding_actor (.HitComponent (Collision|BreakHitResult _array_element))
                  (:IsNotVolume
                    (Variables|Default|SetHitTerrain (Math|Boolean|ORBoolean (not (Variables|/Script/Engine.Actor|GetbCanBeDamaged _asbuilding_actor)) (Variables|/Script/FortniteGame.BuildingActor|GetbStructurallySupportOverlappingActors _asbuilding_actor) (Variables|Default|GetHitTerrain))))))))
          (:then_2
            ;[0.9091 0 1] Make Cosmetic changes (SFX, location, which mesh to show, setting visibility.)
            (Audio|PlaySound2D (Variables|Default|GetGridSnapSound))
            (Transformation|SetWorldLocationAndRotation (Variables|Default|GetBuilding) (LocalVariables|GetGridSnapLocation) (Math|Rotator|MakeRotator 0.0 0.0 (LocalVariables|GetYaw%20Offset)) false true)
            (Components|StaticMesh|SetStaticMesh _sm_buildingpreview (select (Variables|Default|GetHitTerrain) (Variables|Default|GetBuildingPreviewMeshUnderground) (Variables|Default|GetBuildingPreviewMesh)))
            ;[1 1 1] FORT-649417; avoid showing preview unless aiming
            (if (TouchPlatform|IsTouchInputPlatform)
              (bind _asfort_player_pawn (Utilities|Casting|CastToFortPlayerPawn (Game|GetInstigator)))
              (Rendering|SetVisibility _sm_buildingpreview (Math|Boolean|ANDBoolean (GameplayTags|HasMatchingGameplayTag _asfort_player_pawn "(TagName=\"Gameplay.Action.Player.InWindup\")") (not (Class|AthenaHUDContext|UseTaptoShoot (BlueprintContext|GetAthenaHUDContext))) (Variables|Default|GetbIsTrajectoryValid) (not (Variables|Default|GetbWaitFirstTrajectoryValid))) true)
              (else
                (Rendering|SetVisibility _sm_buildingpreview (and (Variables|Default|GetbIsTrajectoryValid) (not (Variables|Default|GetbWaitFirstTrajectoryValid))) true)))))))))

(fn PerformDoorTrace
  (params (Start :type Vector) (End :type Vector) (HalfSize :type Vector) (DrawTime :type Float%20%28single-precision%29) (DrawDebugType :type EDrawDebugTrace%20Enum))
  (bind _outhits (Collision|MultiBoxTraceForObjects :Start Start :End End :HalfSize HalfSize :Orientation "0, 0, 0" :ObjectTypes (Utilities|Array|MakeArray "ObjectTypeQuery1" "ObjectTypeQuery2") :bTraceComplex false :DrawDebugType DrawDebugType :bIgnoreSelf true :TraceColor "(R=1.000000,G=0.000000,B=0.000000,A=1.000000)" :TraceHitColor "(R=0.000000,G=1.000000,B=0.000000,A=1.000000)" :DrawTime DrawTime))
  (return _outhits))

(fn ShouldBounceOffObject
  (params (Hit%20Object :type Actor%20Object%20Reference) (Hit%20Component :type Primitive%20Component%20Object%20Reference) (Hit%20Normal :type Vector))
  (Utilities|IsValid Hit%20Object
    (:Is%20Valid
      (Utilities|IsValid Hit%20Component
        (:Is%20Valid
          (if (or (Utilities|Array|ContainsItem (Variables|/Script/Engine.ActorComponent|GetComponentTags Hit%20Component) (Variables|Default|GetBounceTag)) (Utilities|Array|ContainsItem (Variables|/Script/Engine.Actor|GetTags Hit%20Object) (Variables|Default|GetBounceTag)))
            (return true)
            (else
              (return (Math|Boolean|ORBoolean (Utilities|IsA Hit%20Object "/Script/FortniteGame.FortPawn") (Utilities|IsA Hit%20Object "/Script/FortniteGame.FortAthenaSupplyDrop") (Utilities|IsA Hit%20Object "/Script/FortniteGame.FortAthenaSupplyDropBalloon") (Utilities|IsA Hit%20Object "/Script/FortniteGame.FortProjectileBase") (Utilities|IsA Hit%20Object "/Script/FortniteGame.BuildingContainer") (Utilities|IsA Hit%20Object "/Script/FortniteGame.FortAthenaVehicle"))))))
        (:Is%20Not%20Valid
          _)))
    (:Is%20Not%20Valid
      _))
  (return false))

(fn ConstructionScript
  (|CallParentFunction))
