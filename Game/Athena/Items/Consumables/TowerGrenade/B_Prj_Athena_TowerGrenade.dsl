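; Tower Grenade projectile that deploys a prebuilt tower, facing the thrower's cardinal direction, where it lands.
; It bounces off pawns and select objects, then snaps to the building grid using client-predicted impact within tolerance before exploding.
; Non-cosmetic explosion logic box-traces the doorway, adding a raised base over terrain, then clears cells and spawns budget-checked pieces and tires in delayed stages.
; A timer staggers queued actor destruction; the cosmetic branch plays effects.
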(event EventBeginPlay
  ;[1 1 1] Lookup build delay row. Determine which cardinal direction the tower should face.
  (bind _outxy (DataTable|EvaluateCurveTableRow "/Game/Athena/Balance/DataTables/AthenaGameData.AthenaGameData" "Default.TowerGrenade.BuildDelay")
    (:RowFound
      (Variables|Default|SetBuildDelay _outxy))
    (:RowNotFound
      _))
  (bind _returnvalue (Audio|SpawnSoundAttached :Sound (Variables|Default|GetInAirLoopSound) :AttachToComponent (Variables|Collision|GetCapsuleComponent) :AttachPointName "None" :LocationType "KeepRelativeOffset" :bStopWhenAttachedToDestroyed true :VolumeMultiplier 1.0 :PitchMultiplier 1.0 :StartTime 0.0 :bAutoDestroy true))
  (bind _output_get (Variables|Default|SetInAirLoopComponent _returnvalue))
  (Audio|Components|Audio|FadeIn _output_get 1.0 1.0 0.0 "Linear" :!class /Script/Engine.AudioComponent)
  (CallFunction|CalculateCardinalDirection))

(event EventEndPlay
  (params (EndPlayReason :type EEndPlayReason%20Enum :default "Destroyed"))
  ;[1 1 1] Destroy any remaining actors in our destroy list if needed
  (switch Utilities|FlowControl|Switch|SwitchonEEndPlayReason EndPlayReason
    (:Destroyed
      (for _array_element (Variables|Default|GetActorsToDestroy)
        (Actor|DestroyActor _array_element)))))

(event Collision|EventHit
  (params (MyComp :type Primitive%20Component%20Object%20Reference) (Other :type Actor%20Object%20Reference) (OtherComp :type Primitive%20Component%20Object%20Reference) (bSelfMoved :type Boolean) (HitLocation :type Vector :default "0, 0, 0") (HitNormal :type Vector :default "0, 0, 0") (NormalImpulse :type Vector :default "0, 0, 0") (Hit :type Hit%20Result%20Structure%20%28by%20ref%29))
  ;[1 1 1] Bounce off of pawns and walls
  (bind _returnvalue (CallFunction|ShouldBounceOffObject Other OtherComp HitNormal))
  (if _returnvalue
    (CallFunction|HandleBounce HitLocation)
    (Utilities|FlowControl|DoOnce false
      (:Completed
        (Utilities|FlowControl|Delay (Variables|Default|GetBounceTimeout)))))
  (Utilities|FlowControl|DoOnce false
    (:Completed
      (Game|Components|ProjectileMovement|StopSimulating (Variables|Movement|GetProjectileMovementComponent) Hit))))

(event Custom|Handle%20Bounce
  (params (Hit%20Location :type Vector))
  (bind _lastbouncelocation (Variables|Default|GetLastBounceLocation))
  (bind _bouncesound (Variables|Default|GetBounceSound))
  ;[1 1 1] Play audio feedback, etc when bouncing
  (if (and (> (Math|Vector|VectorLengthSquared (- Hit%20Location (Variables|Default|GetLastBounceLocation))) (Math|Float|Square (Variables|Default|GetMinBounceDistance))) (> (- (Utilities|Time|GetTimeSeconds) (Variables|Default|GetLastBounceTime)) (Variables|Default|GetMinTimeBetweenBounces)))
    (Variables|Default|SetLastBounceLocation Hit%20Location)
    (Variables|Default|SetLastBounceTime (Utilities|Time|GetTimeSeconds))
    (Audio|PlaySoundatLocation _bouncesound _lastbouncelocation)
    (Fortnite|ShowSoundIndicatorLocalforTeam :Location _lastbouncelocation :MaxAudibleDistance (Variables|/Script/Engine.SoundBase|GetMaxDistance _bouncesound) :IndicatorType "Gunshot" :Team (Team|GetActorTeam self) :Affiliations (Utilities|Array|MakeArray "Hostile" "Neutral") :Instigator self :Tint "(R=0.000000,G=0.000000,B=0.000000,A=0.000000)" :AttenuationMaxRangeMultiplier 1.0 :AttenuationIntensityMultiplier 1.0)))

(event EventOnStop
  (params (Hit :type Hit%20Result%20Structure%20%28by%20ref%29))
  (bind (_bblockinghit _binitialoverlap _time _distance _location _impactpoint _normal _impactnormal _physmat _hitactor _hitcomponent) (Collision|BreakHitResult Hit))
  ;[1 1 1] Initializing variables, stop rotation, and snap projectile to grid before "exploding"
  (CallFunction|SetBaseDestination Hit)
  (Components|Movement|SetUpdatedComponent (Variables|Default|GetRotatingMovement) 0)
  (Transformation|SetActorTransform self (Math|Transform|MakeTransform (Fortnite|SnaptoFortBuildingGrid (Math|Vector|SelectVector (Variables|/Game/Athena/Items/Consumables/Parents/B_Prj_Comp_ClientPredicted_Data.B_Prj_Comp_ClientPredicted_Data_C|GetClientPredictedImpactLocation (Variables|Default|GetB_Prj_Comp_ClientPredicted_Data)) (Transformation|GetActorLocation) (Utilities|Operators|Equal(==) (Variables|/Game/Athena/Items/Consumables/Parents/B_Prj_Comp_ClientPredicted_Data.B_Prj_Comp_ClientPredicted_Data_C|GetClientPredictedImpactLocation (Variables|Default|GetB_Prj_Comp_ClientPredicted_Data)) (Transformation|GetActorLocation) (Variables|Default|GetClientLocationErrorTolerance))) (Math|Vector|MakeVector 0.0 0.0 (Variables|SnaptoGrid|GetGridZOffset)) (0.0 0.0 576.0)) (Math|Rotator|MakeRotator 0.0 0.0 (Utilities|Map|Find (Variables|Orientation|GetOrientationYawOffsets) (Variables|Orientation|GetCardinalDirection)))) false false)
  (bind _returnvalue (CallFunction|ShouldBounceOffObject _hitactor _hitcomponent _normal)) ; Same conditions to check for if the building should spawn
  (if _returnvalue
    (bind _reverttopickupwid (Variables|Default|GetRevertToPickupWID)
      (:then
        (Inventory|SpawnPickupInWorld _reverttopickupwid 1 (Transformation|GetActorLocation) "0.000000,0.000000,1.000000" 0 true false false -1 "Tossed" "TossedByPlayer" 0 false)
        (Rendering|SetActorHiddenInGame true)
        (Actor|DestroyActor))
      (:else
        _)))
  (Delivery|DoExplosion))

(event EventOnExploded
  (params (HitActors :type Array%20of%20Actor%20Object%20References) (HitResults :type Array%20of%20Hit%20Result%20Structures))
  (bind _returnvalue (Transformation|GetActorLocation))
  (bind _returnvalue_2 (Transformation|GetActorLocation))
  (bind _returnvalue_3 (- _returnvalue "0, 0,500.000000"))
  (bind _gridexplosionoffset (Variables|Cosmetic|GetGridExplosionOffset))
  (bind _returnvalue_4 (+ _returnvalue_2 _gridexplosionoffset))
  (bind _cue_distantsound (Variables|Cosmetic|GetCue_DistantSound))
  (Networking|CanExecuteCosmeticEvents
    (:True
      ;[1 1 1] OnExploded - Emitter, Sound
      (Effects|SpawnEmitteratLocation (Variables|Cosmetic|GetP_BuildBuilding) _returnvalue)
      (Camera|PlayWorldCameraShake "/Game/Athena/Items/Consumables/TowerGrenade/FX/B_Ground_CameraShake.B_Ground_CameraShake_C" _returnvalue_3 5e+02 2.5e+03 3.0)
      (Fortnite|PlayLocalForceFeedbackatLocationMulti "/Game/Weapons/ForceFeedback/Explosions/FF_Generic_Explosion_Near.FF_Generic_Explosion_Near" "/Game/Weapons/ForceFeedback/Explosions/FF_Generic_Explosion_Far.FF_Generic_Explosion_Far" _returnvalue_3 1e+03 2e+03)
      (Audio|Components|Audio|FadeOut (Variables|Default|GetInAirLoopComponent) 0.2 0.0 "Linear" :!class /Script/Engine.AudioComponent)
      (if (Pawn|IsLocallyControlled (Game|GetInstigator))
        (Audio|PlaySoundatLocation (Variables|Cosmetic|GetCue_CloseSound) _returnvalue_4 "0, 0, 0" 1.0 1.4) ; Play Close Explosion
        (else
          (Audio|PlaySoundatLocation _cue_distantsound _returnvalue_4 "0, 0, 0" 1.0 1.4) ; Play Distant Explosion
        ))
      (Fortnite|ShowSoundIndicatorLocalforTeam :Location _returnvalue_4 :MaxAudibleDistance (Variables|/Script/Engine.SoundBase|GetMaxDistance _cue_distantsound) :IndicatorType "Gunshot" :Team (Team|GetActorTeam self) :Affiliations (Utilities|Array|MakeArray "Hostile" "Neutral") :Instigator self :Tint "(R=0.000000,G=0.000000,B=0.000000,A=0.000000)" :AttenuationMaxRangeMultiplier 1.0 :AttenuationIntensityMultiplier 1.0))
    (:False
      ;[1 1 1] Send a quest event for throwing the TowerGrenade
      (CallFunction|SendQuestEventtoInstigator)
      ;[1 1 1] Trace to see if a player pawn can walk through the door.
      (CallFunction|HandleProps HitActors)
      (bind _outhits (Collision|MultiBoxTraceForObjects (Math|Vector|vector+vector (Variables|Default|GetBaseDestination) (Variables|TerrainTrace|GetDoorTraceOffsetHigh) (Utilities|Map|Find (Variables|Orientation|GetOrientationDoortraceOffsets) (Variables|Orientation|GetCardinalDirection))) (Math|Vector|vector+vector (Variables|Default|GetBaseDestination) (Variables|TerrainTrace|GetDoorTraceOffsetLow) (Utilities|Map|Find (Variables|Orientation|GetOrientationDoortraceOffsets) (Variables|Orientation|GetCardinalDirection))) "48.000000,48.000000,0.000000" "0, 0, 0" (Utilities|Array|MakeArray "ObjectTypeQuery1" "ObjectTypeQuery2") false (Variables|Default|GetActorsToIgnore)))
      ;[1 1 1] Check if any hits are against terrain instead of building actors
      (for _array_element _outhits
        (Development|PrintString (Utilities|String|Append "Tower Grenade Hit: " (Utilities|GetDisplayName (.HitActor (Collision|BreakHitResult _array_element)))) (Variables|Default|GetIsDebug) (Variables|Default|GetIsDebug) "(R=0.000000,G=0.660000,B=1.000000,A=1.000000)" 1e+01)
        (if (Math|Boolean|ORBoolean (Utilities|IsA (.HitActor (Collision|BreakHitResult _array_element)) "/Script/Landscape.Landscape") (Utilities|IsA (.HitActor (Collision|BreakHitResult _array_element)) "/Script/Landscape.LandscapeStreamingProxy") (Utilities|IsA (.HitActor (Collision|BreakHitResult _array_element)) "/Script/FortniteGame.BuildingStaticGeo") (Utilities|IsA (.HitActor (Collision|BreakHitResult _array_element)) "/Script/FortniteGame.FortStaticMeshActor"))
          (Variables|TerrainTrace|Setb_HitTerrain true)
          (else
            (bind _asbuilding_actor (Utilities|Casting|CastToBuildingActor (.HitActor (Collision|BreakHitResult _array_element))))
            (Default|IsNotAVolume _asbuilding_actor (.HitComponent (Collision|BreakHitResult _array_element))
              (:IsNotVolume
                (Variables|TerrainTrace|Setb_HitTerrain (Math|Boolean|ORBoolean (not (Variables|Actor|GetbCanBeDamaged _asbuilding_actor)) (Variables|/Script/FortniteGame.BuildingActor|GetbStructurallySupportOverlappingActors _asbuilding_actor) (Variables|TerrainTrace|Getb_HitTerrain))))))))
      (if (Variables|TerrainTrace|Getb_HitTerrain)
        ;[1 1 1] Door would be underground, so it is moved up a grid cell and we add stairs and archways to the bottom of the tower.
        (Variables|Default|SetBaseDestination (+ (Variables|Default|GetBaseDestination) "0, 0,384.000000"))
        (CallFunction|ClearandBuild (Variables|ClearedCells|GetClearedCells_Underground) (Variables|BGAClasses|GetBGAClasses_Underground) (Variables|BGALocations|GetBGALocations_Underground) (Variables|BGARotations|GetBGARotations_Underground) true))
      ;[1 1 1] OnExploded - Destroy existing building actors that would overlap the tower, then spawn the tower's building actors.
      (CallFunction|ClearandBuild (Variables|ClearedCells|GetClearedCells_0) (Variables|BGAClasses|GetBGAClasses_0) (Variables|BGALocations|GetBGALocations_0) (Variables|BGARotations|GetBGARotations_0) true)
      (Utilities|FlowControl|Delay (Variables|Default|GetBuildDelay))
      (CallFunction|SpawnTires (Variables|Default|GetBaseDestination))
      (CallFunction|ClearandBuild (Variables|ClearedCells|GetClearedCells_1) (Variables|BGAClasses|GetBGAClasses_1) (Variables|BGALocations|GetBGALocations_1) (Variables|BGARotations|GetBGARotations_1) true)
      (Utilities|FlowControl|Delay (Variables|Default|GetBuildDelay))
      (CallFunction|ClearandBuild (Variables|ClearedCells|GetClearedCells_2) (Variables|BGAClasses|GetBGAClasses_2) (Variables|BGALocations|GetBGALocations_2) (Variables|BGARotations|GetBGARotations_2))
      (Utilities|FlowControl|Delay (Variables|Default|GetBuildDelay))
      (CallFunction|ClearandBuild (Variables|ClearedCells|GetClearedCells_3) (Variables|BGAClasses|GetBGAClasses_3) (Variables|BGALocations|GetBGALocations_3) (Variables|BGARotations|GetBGARotations_3))
      ;[1 1 1] loop a timer deleting an actor every frame or so
      (bind _returnvalue_5 (Utilities|Time|SetTimerbyEvent (delegate Custom|Delete%20another%20actor) 0.03 true)))))

(event Custom|Delete%20another%20actor
  (bind _actorstodestroy (Variables|Default|GetActorsToDestroy))
  ;[1 1 1] loop a timer deleting an actor every frame or so
  (if (> (Utilities|Array|Length _actorstodestroy) 0)
    (bind (_array_element _array_index) (Utilities|Array|ForEachLoopwithBreak _actorstodestroy :!label _label)
      (:LoopBody
        (Utilities|IsValid _array_element
          (:Is%20Valid
            (if (Variables|Actor|GetbCanBeDamaged _array_element)
              (Actor|DestroyActor _array_element)))
          (:Is%20Not%20Valid
            _))
        (Utilities|Array|RemoveIndex _actorstodestroy _array_index)
        (goto _label Break)))
    (else
      (Utilities|Time|ClearandInvalidateTimerbyHandle (stale EventOnExploded _returnvalue_5)))))

(fn CalculateCardinalDirection
  (bind _returnvalue (Game|GetInstigator))
  (bind _returnvalue_2 (Transformation|GetActorForwardVector _returnvalue))
  (bind _x (.x _returnvalue_2))
  (bind _y (.y _returnvalue_2))
  (if (>= (Math|Float|Absolute(Float) _x) (Math|Float|Absolute(Float) _y))
    (if (>= _x 0.0)
      (Variables|Orientation|SetCardinalDirection "NewEnumerator0")
      (else
        (Variables|Orientation|SetCardinalDirection "NewEnumerator1")))
    (elif (>= _y 0.0)
      (Variables|Orientation|SetCardinalDirection "NewEnumerator2")
      (else
        (Variables|Orientation|SetCardinalDirection "NewEnumerator3")))))

(fn CanDestroyObject
  (params (Object :type Actor%20Object%20Reference))
  (bind _object (Variables|GetObject))
  (bind _bcanbedamaged (Variables|Actor|GetbCanBeDamaged _object))
  (bind _asbuilding_actor (Utilities|Casting|CastToBuildingActor Object)
    (:then
      (LocalVariables|SetBuilding _asbuilding_actor)
      (bind _returnvalue (CallFunction|GetOwnerWeapon))
      (Utilities|IsValid _returnvalue
        (:Is%20Valid
          (Utilities|IsValid (Actor|GetOwner _returnvalue)
            (:Is%20Valid
              (bind _aspawn (Utilities|Casting|CastToPawn (Actor|GetOwner _returnvalue))
                (:then
                  (bind _returnvalue_2 (Class|BuildingActor|CanDestroyBuildingActor (LocalVariables|GetBuilding) (Pawn|GetController _aspawn) _aspawn))
                  (return _returnvalue_2))
                (:CastFailed
                  (return _bcanbedamaged))))
            (:Is%20Not%20Valid
              (return _bcanbedamaged))))
        (:Is%20Not%20Valid
          (return _bcanbedamaged))))
    (:CastFailed
      (return (Variables|Actor|GetbCanBeDamaged (Variables|GetObject))))))

(fn ClearAndBuild
  (params (ClearedGridCells :type Array%20of%20Vectors) (BGAClassArray :type Array%20of%20Actor%20Class%20References) (BGALocationArray :type Array%20of%20Vectors) (BGARotationArray :type Array%20of%20Rotators) (DestroyFloorsQuickly :type Boolean))
  (for _array_element ClearedGridCells
    (CallFunction|DestroyActorsinGridCell (+ (Variables|Default|GetBaseDestination) (Math|Vector|RotateVectorAroundAxis _array_element (Utilities|Map|Find (Variables|Orientation|GetOrientationYawOffsets) (Variables|Orientation|GetCardinalDirection)) "0, 0,1.000000")) DestroyFloorsQuickly))
  (CallFunction|CreateBaseSection (Variables|Default|GetBaseDestination) BGAClassArray BGALocationArray BGARotationArray))

(fn CreateBaseSection
  (params (ReferenceLocation :type Vector) (BGAClassArray :type Array%20of%20Actor%20Class%20References) (BGALocationArray :type Array%20of%20Vectors) (BGARotationArray :type Array%20of%20Rotators))
  (bind _returnvalue (Actor|GetOwner))
  (if (> (Utilities|Array|Length BGAClassArray) 0)
    (if (Creative|Budget|CanPlaceClassInstances (Fortnite|GetFortPlayerControllerfromActor (Actor|GetOwner)) BGAClassArray ReferenceLocation)
      (for _index (range (Utilities|Array|Length BGAClassArray))
        (bind _asbuilding_actor (Utilities|Casting|CastToBuildingActorClass (Utilities|Array|Get(acopy) BGAClassArray _index)))
        (bind _returnvalue_2 (Building|SpawnBuildingActor _asbuilding_actor (Math|Transform|MakeTransform (Math|Vector|vector+vector ReferenceLocation (Math|Vector|RotateVectorAroundAxis (Utilities|Array|Get(aref) BGALocationArray _index) (Utilities|Map|Find (Variables|Orientation|GetOrientationYawOffsets) (Variables|Orientation|GetCardinalDirection)) "0, 0,1.000000") (Utilities|Map|Find (Variables|Orientation|GetOrientationLocationOffsets) (Variables|Orientation|GetCardinalDirection))) (Math|Rotator|MakeRotator 0.0 0.0 (+ (.yaw (Utilities|Array|Get(aref) BGARotationArray _index)) (Utilities|Map|Find (Variables|Orientation|GetOrientationYawOffsets) (Variables|Orientation|GetCardinalDirection)))))))
        (BuildingActor|SetPlayerPlacedforDestruction _returnvalue_2 true)
        (Class|BuildingActor|SetTeam _returnvalue_2 (Team|GetActorTeam _returnvalue))
        (Building|InitializeKismetSpawnedBuildingActor _returnvalue_2 0 (Fortnite|GetFortPlayerControllerfromActor _returnvalue) false 0 false)
        (if (not (and (Building|IsWorldLocValid (Building|GetBuildingStructuralSupportSystem) (+ (Transformation|GetActorLocation _returnvalue_2) "0, 0,384.000000")) (Building|IsWorldLocValid (Building|GetBuildingStructuralSupportSystem) (+ (Transformation|GetActorLocation _returnvalue_2) "0, 0,-384.000000"))))
          (Actor|DestroyActor _returnvalue_2))))))

(fn DestroyActorsInGridCell
  (params (CellLocation :type Vector%20%28by%20ref%29) (DestroyFloorsQuickly :type Boolean))
  ;[1 1 1] Destroys all building actors in a grid cell given a world location
  (bind _outactorsingridcell (Building|K2GetBuildingActorsinGridCell (Building|GetBuildingStructuralSupportSystem) CellLocation (Utilities|Struct|MakeBuildingGridActorFilter true)))
  ;;[1 1 1] Walls
  (for _array_element (Utilities|Struct|BreakBuildingNeighboringActorInfo _outactorsingridcell)
    (CallFunction|TryDestroyActor (Utilities|Struct|BreakNeighboringWallInfo _array_element)))
  ;;[1 1 1] Floors
  (for _array_element_2 (.NeighboringFloorInfos (Utilities|Struct|BreakBuildingNeighboringActorInfo _outactorsingridcell))
    (CallFunction|TryDestroyActor (Utilities|Struct|BreakNeighboringFloorInfo _array_element_2) (Variables|GetDestroyFloorsQuickly)))
  ;;[1 1 1] Center Cells
  (for _array_element_3 (.NeighboringCenterCellInfos (Utilities|Struct|BreakBuildingNeighboringActorInfo _outactorsingridcell))
    (CallFunction|TryDestroyActor (Utilities|Struct|BreakNeighboringCenterCellInfo _array_element_3) (Variables|GetDestroyFloorsQuickly))))

(fn HandleProps
  (params (Array :type Array%20of%20Actor%20Object%20References))
  (for _array_element Array
    (bind _asbuilding_prop (Utilities|Casting|CastToBuildingProp _array_element)
      (:then
        (bind _returnvalue (GameplayTags|HasMatchingGameplayTag(Message) _array_element "(TagName=\"Building.Type.Container.TreasureChest.Common\")")) ; Look for the gameplay tag this BP actor has before doing the potentally \nexpensive IsA check
        (if _returnvalue
          (Utilities|IsA(soft) _array_element "/Game/Building/ActorBlueprints/Containers/Tiered_Chest_Athena.Tiered_Chest_Athena_C"
            (:True
              _)
            (:False
              (if (Utilities|IsA _array_element "/Script/FortniteGame.BuildingPropBigHealth") :!label _label
                _
                (elif (Variables|/Script/FortniteGame.BuildingActor|GetbStructurallySupportOverlappingActors _asbuilding_prop)
                  _
                  (else
                    (if (Variables|Actor|GetbCanBeDamaged _array_element) :!label _label_3
                      (Actor|DestroyActor _array_element)
                      (else
                        _)))))))
          (Utilities|Array|Add (Variables|Default|GetActorsToIgnore) _array_element :!label _label_2))
        (goto _label))
      (:CastFailed
        (if (Utilities|IsA _array_element "/Game/Athena/SupplyDrops/BP_DamageBalloon_Athena.BP_DamageBalloon_Athena_C")
          _
          (elif (Utilities|IsA _array_element "/Script/FortniteGame.BuildingActor")
            (goto _label_2)))))
    (goto _label_3)))

(fn SendQuestEventToInstigator
  ; Send a quest event for placing the TowerGrenade
  (CustomVerb|BroadcastCustomVerbMessage (CustomVerb|MakeCustomVerbMessage (Game|GetInstigator) "(TagName=\"CustomVerb.DeployTowerGrenade\")")))

(fn SetBaseDestination
  (params (Hit :type Hit%20Result%20Structure%20%28by%20ref%29))
  (bind (_bblockinghit _binitialoverlap _time _distance _location _impactpoint _normal _impactnormal) (Collision|BreakHitResult Hit))
  (Networking|SwitchHasAuthority
    (:Authority
      (Variables|Default|SetBaseDestination (Fortnite|SnaptoFortBuildingGrid (+ (Math|Vector|SelectVector (Math|Vector|MakeVector (.x (Variables|/Game/Athena/Items/Consumables/Parents/B_Prj_Comp_ClientPredicted_Data.B_Prj_Comp_ClientPredicted_Data_C|GetClientPredictedImpactLocation (Variables|Default|GetB_Prj_Comp_ClientPredicted_Data))) (.y (Variables|/Game/Athena/Items/Consumables/Parents/B_Prj_Comp_ClientPredicted_Data.B_Prj_Comp_ClientPredicted_Data_C|GetClientPredictedImpactLocation (Variables|Default|GetB_Prj_Comp_ClientPredicted_Data))) (.z _impactpoint)) _impactpoint (Utilities|Operators|Equal(==) (Variables|/Game/Athena/Items/Consumables/Parents/B_Prj_Comp_ClientPredicted_Data.B_Prj_Comp_ClientPredicted_Data_C|GetClientPredictedImpactLocation (Variables|Default|GetB_Prj_Comp_ClientPredicted_Data)) (Transformation|GetActorLocation) (Variables|Default|GetClientLocationErrorTolerance))) (* _impactnormal 1)) (Math|Vector|MakeVector 0.0 0.0 (Variables|SnaptoGrid|GetGridZOffset)) (Math|Vector|MakeVector 0.0 (Variables|SnaptoGrid|GetGridYOffset)))))))

(fn ShouldBounceOffObject
  (params (Hit%20Object :type Actor%20Object%20Reference) (Hit%20Component :type Primitive%20Component%20Object%20Reference) (Hit%20Normal :type Vector))
  (Utilities|IsValid Hit%20Object
    (:Is%20Valid
      (Utilities|IsValid Hit%20Component
        (:Is%20Valid
          (if (or (Utilities|Array|ContainsItem (Variables|/Script/Engine.ActorComponent|GetComponentTags Hit%20Component) (Variables|Default|GetBounceTag)) (Utilities|Array|ContainsItem (Variables|Actor|GetTags Hit%20Object) (Variables|Default|GetBounceTag)))
            (return true)
            (else
              (return (Math|Boolean|ORBoolean (Utilities|IsA Hit%20Object "/Script/FortniteGame.FortPawn") (Utilities|IsA Hit%20Object "/Script/FortniteGame.FortAthenaSupplyDrop") (Utilities|IsA Hit%20Object "/Script/FortniteGame.FortAthenaSupplyDropBalloon") (Utilities|IsA Hit%20Object "/Script/FortniteGame.FortProjectileBase") (Utilities|IsA Hit%20Object "/Script/FortniteGame.BuildingContainer") (Utilities|IsA Hit%20Object "/Script/FortniteGame.FortAthenaVehicle"))))))
        (:Is%20Not%20Valid
          _)))
    (:Is%20Not%20Valid
      _))
  (return false))

(fn SpawnTires
  (params (ReferenceLocation :type Vector))
  (bind _cardinaldirection (Variables|Orientation|GetCardinalDirection))
  (bind _orientationlocationoffsets (Variables|Orientation|GetOrientationLocationOffsets))
  (bind _value (Utilities|Map|Find _orientationlocationoffsets _cardinaldirection))
  (bind _cardinaldirection_2 (Variables|Orientation|GetCardinalDirection))
  (bind _orientationyawoffsets (Variables|Orientation|GetOrientationYawOffsets))
  (bind _value_2 (Utilities|Map|Find _orientationyawoffsets _cardinaldirection_2))
  (bind _bgalocations_tires (Variables|BGALocations|GetBGALocations_Tires))
  (bind _bgaclasses_tires (Variables|BGAClasses|GetBGAClasses_Tires))
  (if (> (Utilities|Array|Length _bgaclasses_tires) 0)
    (if (Creative|Budget|CanPlaceClassInstances (Fortnite|GetFortPlayerControllerfromActor (Actor|GetOwner)) _bgaclasses_tires ReferenceLocation)
      (for _index (range (Utilities|Array|Length (Variables|BGAClasses|GetBGAClasses_Tires)))
        (bind (_outhit _returnvalue) (Collision|SphereTraceByChannel (+ (Math|Vector|vector+vector ReferenceLocation (Math|Vector|RotateVectorAroundAxis (Utilities|Array|Get(aref) _bgalocations_tires _index) _value_2 "0, 0,1.000000") _value) "0, 0,384.000000") (Math|Vector|vector+vector ReferenceLocation (Math|Vector|RotateVectorAroundAxis (Utilities|Array|Get(aref) _bgalocations_tires _index) _value_2 "0, 0,1.000000") _value) 32.0 "TraceTypeQuery10" false (Variables|Default|GetActorsToIgnore)))
        (Development|PrintString (Utilities|String|Append "Tower Grenade Tire trace hit: " (Utilities|GetDisplayName (.HitActor (Collision|BreakHitResult _outhit)))) (Variables|Default|GetIsDebug) (Variables|Default|GetIsDebug))
        (bind _asbuilding_actor (Utilities|Casting|CastToBuildingActorClass (Utilities|Array|Get(acopy) (Variables|BGAClasses|GetBGAClasses_Tires) _index)))
        (bind _returnvalue_2 (Building|SpawnBuildingActor _asbuilding_actor (Math|Transform|MakeTransform (select _returnvalue (- (.Location (Collision|BreakHitResult _outhit)) "0, 0,32.000000") (Math|Vector|vector+vector ReferenceLocation (Math|Vector|RotateVectorAroundAxis (Utilities|Array|Get(aref) _bgalocations_tires _index) _value_2 "0, 0,1.000000") _value)) (Math|Rotator|MakeRotator 0.0 0.0 (+ (.yaw (Utilities|Array|Get(aref) (Variables|BGARotations|GetBGARotations_Tires) _index)) _value_2)))))
        (Building|InitializeKismetSpawnedBuildingActor _returnvalue_2 0 0 false 0 false)
        (BuildingActor|SetPlayerPlacedforDestruction _returnvalue_2 true)
        (if (not (and (Building|IsWorldLocValid (Building|GetBuildingStructuralSupportSystem) (+ (Transformation|GetActorLocation _returnvalue_2) "0, 0,384.000000")) (Building|IsWorldLocValid (Building|GetBuildingStructuralSupportSystem) (+ (Transformation|GetActorLocation _returnvalue_2) "0, 0,-384.000000"))))
          (Actor|DestroyActor _returnvalue_2))))))

(fn TryDestroyActor
  (params (InActor :type Building%20Static%20Mesh%20Object%20Reference) (Quick%20Destroy :type Boolean))
  (bind _inactor (Variables|GetInActor))
  (bind _returnvalue (CallFunction|CanDestroyObject InActor))
  (if _returnvalue
    (if (Variables|GetQuick%20Destroy)
      (Actor|DestroyActor (Variables|GetInActor))
      (else
        (Utilities|Array|Add (Variables|Default|GetActorsToDestroy) _inactor))))
  (Utilities|Array|Add (Variables|Default|GetActorsToIgnore) _inactor))

(fn ConstructionScript
  (Networking|CanExecuteCosmeticEvents
    (:True
      (Components|Movement|SetUpdatedComponent (Variables|Default|GetRotatingMovement) (Variables|Default|GetMesh)))))
