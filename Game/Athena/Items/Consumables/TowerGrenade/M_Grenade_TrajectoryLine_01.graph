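# Unlit, two-sided translucent spline-mesh material for the Tower Grenade's trajectory line.
# Panning dashes tile by the StartLocation-to-EndLocation distance and blend base and bright colors, with a faster-falling strobe pulse added on top.
# Color is strengthened at the ribbon's V edges and scaled by inverted Fresnel, while opacity fades along U.
# A world-distance falloff from StartLocation is also computed.
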Material /Game/Athena/Items/Consumables/TowerGrenade/M_Grenade_TrajectoryLine_01
graph {
 properties {
  bUsedWithSplineMeshes=true
  blendMode=BLEND_Translucent
  pixelDepthOffsetMode=PDOM_Legacy
  shadingModel=MSM_Unlit
  twoSided=true
  usedShadingModels=Unlit
 }
 nodes [
  n1 = TextureCoordinate()
  n2 = VectorParameter() {parameterName=StartLocation}
  n3 = VectorParameter() {parameterName=EndLocation, defaultValue=<rgba 0 0 1 0>}
  n4 = MaterialFunctionCall(
   X=Multiply(A=ScalarParameter() {defaultValue=0.0156, parameterName=ScanlineTileHorizontal}, B=Distance(A=@n2.RGB, B=@n3.RGB)) {constB=16}
   Y=Constant() {r=1}
  ) {materialFunction=&MakeFloat2}
  n5 = ComponentMask(None=Panner(Coordinate=Multiply(A=@n1, B=@n4.Result)) {speedX=-5}) {r=true}
  n6 = Floor(None=Add(A=Frac(None=@n5)) {constB=0.5})
  n7 = Panner(Coordinate=TextureCoordinate() {uTiling=0.6}) {speedX=-1}
  n8 = Frac(None=Multiply(A=ComponentMask(None=@n7) {r=true}))
  n9 = Saturate(None=Subtract(A=Multiply(A=@n8) {constB=8}) {constB=7})
  n10 = Add(A=Add(A=@n6) {constB=0.8}, B=@n9)
  n11 = OneMinus(None=ComponentMask(None=TextureCoordinate()) {r=true})
  Multiply(A=Saturate(None=Multiply(A=@n10, B=@n11)), B=ScalarParameter() {defaultValue=0.3, parameterName=Opacity})
  n12 = ComponentMask(None=TextureCoordinate()) {g=true}
  n13 = Multiply(A=@n12, B=OneMinus(None=@n12))
  n14 = Power(Base=OneMinus(None=Multiply(A=@n13) {constB=4})) {constExponent=4}
  n15 = Add(A=Floor(None=Add(A=@n14) {constB=0.8})) {constB=0.12}
  n16 = LinearInterpolate(
   A=VectorParameter() {parameterName=Base Color, defaultValue=<rgba 4 0 0 0>}.RGB
   B=VectorParameter() {parameterName=Bright Color, defaultValue=<rgba 4 0 0 0>}.RGB
   Alpha=@n6
  )
  n17 = VectorParameter() {parameterName=Strobe Color, defaultValue=<rgba 2 0.2982 0 0>}
  n18 = MaterialFunctionCall(Exponent=Constant() {r=1.5}) {materialFunction=&MF_Fresnel_FX}
  Multiply(A=Add(A=Multiply(A=@n15, B=@n16), B=Multiply(A=@n9, B=@n17.RGB)), B=Add(A=Multiply(A=@n18.Fresnel Inverted) {constB=3}) {constB=0.3})
  n19 = WorldPosition()
  n20 = VectorParameter() {parameterName=StartLocation}
  n21 = Divide(
   A=Subtract(A=Distance(A=@n19.XYZ, B=@n20.RGB), B=ScalarParameter() {defaultValue=256, parameterName=StartDistance})
   B=ScalarParameter() {defaultValue=256, parameterName=BlendDistance}
  )
  Power(Base=OneMinus(None=Saturate(None=@n21)))
 ]
}
