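; Tire pile prop, under the TowerGrenade consumable, that bounces pawns upward when they enter its Jump_Collision.
; Overlapping pawns not already launched or DBNO are detached from remote-controlled pawns if player pawns, launched, and given the PlayerLaunched gameplay effect.
; A capsule trace above the prop checks for a ceiling, choosing obstructed or unobstructed bounce height.
; Controller force feedback plays, and a DoOnce with delayed reset debounces the tire sound.
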(event Custom|Reset%20Bounce%20Sound
  ;[1 1 1] Keeps bounce sound from double triggering (hopefully)
  (Utilities|FlowControl|DoOnce false
    (:Completed
      (Audio|PlaySoundatLocation (Variables|Default|GetTireSound) (Transformation|GetActorLocation (Variables|Default|GetOverlapPawn))))))

(event OnComponentBeginOverlap(Jump_Collision)
  (params (OverlappedComponent :type Primitive%20Component%20Object%20Reference) (OtherActor :type Actor%20Object%20Reference) (OtherComp :type Primitive%20Component%20Object%20Reference) (OtherBodyIndex :type Integer) (bFromSweep :type Boolean) (SweepResult :type Hit%20Result%20Structure%20%28by%20ref%29))
  (bind _overlapcontroller (Variables|Default|GetOverlapController))
  (bind _asfort_pawn (Utilities|Casting|CastToFortPawn OtherActor))
  (bind _output_get (Variables|Default|SetOverlapPawn _asfort_pawn))
  (if (not (or (GameplayTags|HasMatchingGameplayTag (Ability|GetAbilitySystemComponent _output_get) "(TagName=\"Gameplay.Action.Player.Launched\")") (Variables|/Script/FortniteGame.FortPawn|GetbIsDBNO _output_get)))
    (bind _asfort_player_controller (Utilities|Casting|CastToFortPlayerController (Pawn|GetController _output_get))
      (:then
        (Variables|Default|SetOverlapController _asfort_player_controller))
      (:CastFailed
        _))
    ;[1 1 1] Ceiling check
    (bind (_outhit _returnvalue) (Collision|CapsuleTraceByChannel (Math|Vector|MakeVector (.x (Transformation|GetWorldLocation (Variables|Transformation|GetRootComponent))) (.y (Transformation|GetWorldLocation (Variables|Transformation|GetRootComponent))) (+ (.z (Transformation|GetWorldLocation (Variables|Transformation|GetRootComponent))) 200)) (Math|Vector|MakeVector (.x (Transformation|GetWorldLocation (Variables|Transformation|GetRootComponent))) (.y (Transformation|GetWorldLocation (Variables|Transformation|GetRootComponent))) (+ (.z (Transformation|GetWorldLocation (Variables|Transformation|GetRootComponent))) 340)) 50 100 "TraceTypeQuery9" false (Utilities|Array|MakeArray (Variables|Default|GetOverlapPawn))))
    (bind _output_get_2 (Variables|Default|SetCeiling _returnvalue))
    (if _output_get_2
      (Variables|Default|SetBounceHeight 0.0)
      (else
        (Variables|Default|SetBounceHeight 0.0)))
    (bind _asfort_player_pawn (Utilities|Casting|CastToFortPlayerPawn (Variables|Default|GetOverlapPawn))
      (:then
        (RemoteControlPawn|ForceDetachfromRemoteControlledPawn _asfort_player_pawn))
      (:CastFailed
        _))
    (Character|LaunchCharacter (Variables|Default|GetOverlapPawn) (Math|Vector|MakeVector 0.0 0.0 (+ (* (.z (Transformation|GetVelocity)) -1) (select _returnvalue (Variables|Default|GetBounceHeightObstructed) (Variables|Default|GetBounceHeightUnobstructed)))) false true)
    (GameplayEffects|ApplyGameplayEffectToSelf :self (Ability|GetAbilitySystemComponent (Variables|Default|GetOverlapPawn)) :GameplayEffectClass (Variables|Default|GetGE_PlayerLaunched) :Level 0.0)
    (Utilities|FlowControl|Sequence
      (:then_0
        ;[1 1 1] Keeps bounce sound from double triggering (hopefully)
        (Utilities|FlowControl|DoOnce false
          (:Completed
            (Audio|PlaySoundatLocation (Variables|Default|GetTireSound) (Transformation|GetActorLocation (Variables|Default|GetOverlapPawn))))))
      (:then_1
        (Utilities|IsValid _overlapcontroller
          (:Is%20Valid
            (Game|Feedback|LocalPlayForceFeedback _overlapcontroller (Variables|Default|GetLaunchForceFeedback) false "TireLaunch"))
          (:Is%20Not%20Valid
            _))
        (Variables|Default|SetOverlapPawn 0)
        (Variables|Default|SetOverlapController 0)
        ;[1 1 1] Reset Bounce sound, so sound cue can be heard if "X" times bounce happens
        (Utilities|FlowControl|Delay 0.1)
        (CallFunction|ResetBounceSound)))))

(fn ConstructionScript)
